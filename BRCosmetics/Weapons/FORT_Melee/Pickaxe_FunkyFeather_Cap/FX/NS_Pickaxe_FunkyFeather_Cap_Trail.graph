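# Niagara swing trail for the FunkyFeather_Cap pickaxe, combining ribbon anim trails with drifting mesh "Notes" particles.
# TrailHead, TrailBody and TrailEyes draw short ribbons from socket or bone indices, scaled by time-of-day color and system fades.
# Notes spawn around a sampled socket, inherit velocity, pick a random mesh, then spin, scale and wander under curl noise.
# An integer variant masks trail colors and opacities, note color and note visibility tag.
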NiagaraSystem /BRCosmetics/Weapons/FORT_Melee/Pickaxe_FunkyFeather_Cap/FX/NS_Pickaxe_FunkyFeather_Cap_Trail
dependencies {
 usedDataInterfaces [
  &NiagaraDataInterfaceColorCurve
  &NiagaraDataInterfaceArrayFloat
  &NiagaraDataInterfaceArrayColor
  &NiagaraDataInterfaceMeshRendererInfo
  &NiagaraDataInterfaceVectorCurve
  &NiagaraDataInterfaceCurve
  &NiagaraDataInterfaceSocketReader
 ]
 usedDynamicInputs [
  &MergeHelper_FloatSwitch
  &FNTimeOfDayColor
  &ClampFloat
  &RemapRange
  &Add_Float
  &Subtract_Float
  &MakeIntFromFloat
 ]
 usedModules [
  &self:SystemSpawnScript.SystemScriptSource.SystemScriptGraph.NiagaraNodeAssignment_1.SetVariables_#1
  &self:SystemSpawnScript.SystemScriptSource.SystemScriptGraph.NiagaraNodeAssignment_3.SetVariables_#2
  &SystemState
  &self:SystemSpawnScript.SystemScriptSource.SystemScriptGraph.NiagaraNodeAssignment_9.SetVariables_#3
  &self:SystemSpawnScript.SystemScriptSource.SystemScriptGraph.NiagaraNodeAssignment_6.SetVariables_#4
  &self:SystemSpawnScript.SystemScriptSource.SystemScriptGraph.NiagaraNodeAssignment_4.SetVariables_#5
  &self:SystemSpawnScript.SystemScriptSource.SystemScriptGraph.NiagaraNodeAssignment_8.SetVariables_#6
 ]
}
emitters {
 Notes {
  emitterUpdateScript [
   &EmitterState(Life Cycle Mode=::NewEnumerator1, Loop Behavior=::NewEnumerator1, Loop Delay=0.1, Loop Duration=0.2, UseLoopDelay=true)
   &SpawnRate(Spawn Probability=0.3, SpawnRate=30, Use Spawn Probability=false) {module=SpawnRate001}
  ]
  particleSpawnScript [
   &InitializeParticle(
    Color=&MaskLinearColor(
     Mask Source=1
     Mask Parameter=$System.intVariant
     Mask=0
     Masked Linear Color=&SelectLinearColorFromArray(
      Array Sampling Mode=::NewEnumerator0
      Direct Array Index=?
      Color Selection Array=DataInterface() {propertyValues {ColorData [<rgba 0.36 0.9 0.5535 1>, <rgba 0.8019 0.5 1 1>, <rgba 1 0.5127 0.3 1>]}}
      Random Seed=0
     )
     Passthrough Linear Color=<rgba 0.7826 0.5711 0.2586 1>
    )
    Lifetime=&RandomRangeFloat(Minimum=0.4, Maximum=0.55)
    Mesh Scale Mode=::NewEnumerator3
    UsePositionOffset=true
    Write Material Random=true
   )
   /SetVariables(Particles.VisibilityTag=/Niagara/DynamicInputs/Helpers/MaskInt(
     Mask Source:NiagaraInt32=1
     Mask Parameter:NiagaraInt32=$System.intVariant
     Mask:NiagaraInt32=0
     Masked Int:NiagaraInt32=0
     Passthrough Int:NiagaraInt32=1
    )) {isSetParameters=true}
   &SocketLocation(
    Interpolate Socket Sampling=true
    SamplePosOffset=false
    Socket Index=4
    Socket Reader=$System.Socket Reader
    Socket Sampling Mode=::NewEnumerator3
    VelocitySpeedLimit=false
   )
   &ShapeLocation(MaskSpawnGroup=false, Override Seed=false, Sphere Radius=15, UseOwnerScale=true)
   &InitialMeshOrientation(
    Mesh Orientation Mode=::NewEnumerator6
    Orientation Quaternion:Quat4f=$Output.SocketLocation.SampledRotation
    Rotation Coordinate Space=::NewEnumerator2
    Use Rotation Vector=true
   )
   &CameraOffset(Camera Offset Amount=12)
   &InheritVelocity(Inherited Velocity Amount Scale=&VectorFromFloat(Value=1), Speed Limit=true, UseSpeedThreshold=false)
   &AddVelocity(Velocity=&RandomVector(Vector Scale=1), Velocity Speed Scale=22)
   /SetVariables(Particles.MeshIndex=&RandomRangeInt(Minimum=0, Maximum=3)) {isSetParameters=true}
  ]
  particleUpdateScript [
   &ParticleState()
   &ScaleMeshSize(Scale Factor=&Multiply_VectorByFloat(
     Vector=&VectorFromCurve(
      VectorCurve=DataInterface() {propertyValues {XCurve {keys [<at 0 1>,<at 1 1>]}, YCurve {keys [<at 0 1>,<at 1 1>]}, ZCurve {keys [<at 0 0.1>,<at 0.4 1>]}}}
      CurveIndex=$Particles.Age
     )
     Float=&FloatFromCurve(FloatCurve=DataInterface() {propertyValues {Curve {
         defaultValue=0
         keys [
          {arriveTangent=9.0725, interpMode=RCIM_Cubic, leaveTangent=9.0725, tangentMode=RCTM_User, time=0, value=0}
          {arriveTangent=0.3041, interpMode=RCIM_Cubic, leaveTangent=0.3041, tangentMode=RCTM_User, time=0.2053, value=0.9847}
          {arriveTangent=-1.239, interpMode=RCIM_Cubic, leaveTangent=-1.239, time=0.7473, value=0.9144}
          {arriveTangent=-0.0405, interpMode=RCIM_Cubic, leaveTangent=-0.0405, tangentMode=RCTM_User, time=1, value=0}
         ]
        }}})
    ))
   &CurlNoiseForce(
    Bias Noise=true
    Bias Noise Field:Vector3f=&RandomVector(Vector Scale=1)
    Noise Frequency=1
    Noise Quality / Cost=::NewEnumerator0
    Noise Strength=&RandomRangeFloat(Minimum=12, Maximum=60)
    Pan Noise=true
    Pan Noise Field=<0 0 0.2>
   )
   &UpdateMeshOrientation(
    Rotation Rate=&FloatFromCurve(
     FloatCurve=DataInterface() {propertyValues {Curve {
        defaultValue=1
        keys [
         <at 0.0838 0 RCIM_Cubic>
         {arriveTangent=4.6559, interpMode=RCIM_Cubic, leaveTangent=4.6559, tangentMode=RCTM_User, time=0.1893, value=0.4949}
         {arriveTangent=0.4937, interpMode=RCIM_Cubic, leaveTangent=0.4937, tangentMode=RCTM_User, time=0.5019, value=1}
        ]
       }}}
     Scale Curve=&RandomRangeFloat(Minimum=0.8, Maximum=1.4)
    )
    Rotation Vector=&RandomVector(Vector Scale=1)
   )
   &SolveForcesAndVelocity(Clamp Velocity=false, Limit Acceleration=false)
  ]
  rendererClasses [&NiagaraMeshRendererProperties]
  renderers [
   {rendererClass=&NiagaraMeshRendererProperties, rendererIndex=0}
   {rendererClass=&NiagaraMeshRendererProperties, rendererIndex=1}
  ]
 }
 TrailBody {
  emitterUpdateScript [
   &EmitterState()
   &FN_SpawnAnimTrail(Number of Anim Trails=2)
  ]
  particleSpawnScript [
   &FN_InitializeAnimTrailBoneIndecies(First Bone or Socket Index=2)
   &FN_InitializeAnimTrailBoneIndecies(AnimTrail Selection=1, First Bone or Socket Index=3, Second Bone or Socket Index=2) {module=FN_InitializeAnimTrailBoneIndecies001}
   &FN_InitializeAnimTrail(
    Color=&MakeLinearColorFromVectorAndFloat(
     "Vector (RGB)"=&MakeVectorFromLinearColorRGB(Color=&MaskLinearColor(
       Mask Source=1
       Mask Parameter=$System.intVariant
       Mask=0
       Masked Linear Color=&MaskLinearColorBySpawnGroup(Masked Linear Color=<rgba 0 0.159 0.5333 1>, Passthrough Linear Color=<rgba 0.0381 0.0465 0.1043 1>)
       Passthrough Linear Color=&MaskLinearColorBySpawnGroup(Masked Linear Color=<rgba 0.7812 0.5172 0.0611 1>, Passthrough Linear Color=<rgba 0.016 0.0185 0.0545 1>)
      ))
     "Float (Alpha)"=&Multiply_Float(
      A=&MaskFloat(
       Mask Source=1
       Mask Parameter=$System.intVariant
       Mask=0
       Masked Float=&SelectFloatFromArray(Direct Array Index=&SpawnGroup(), Float Selection Array=DataInterface() {propertyValues {FloatData [2,1.4]}})
       Passthrough Float=&SelectFloatFromArray(Direct Array Index=&SpawnGroup(), Float Selection Array=DataInterface() {propertyValues {FloatData [0.5,0.8]}})
      )
      B=$System.FadeSpawn
     )
    )
    Lifetime=0.12
   )
  ]
  particleUpdateScript [
   &FN_UpdateAnimTrail() {module=FN_UpdateAnimTrail001}
   &ScaleColor(
    Scale Mode=1
    Scale RGBA=&Multiply_Vector4(
     A=&MakeVector4_FromLinearColor(LinearColor=$System.TODColor)
     B=&MakeVector4FromVectorAndFloat("Vector (RGB)"=<1 1 1>, "Float (Alpha)"=$System.FadeSystem)
    )
    ScaleRGBA=true
   ) {module=ScaleColor002}
   &ScaleColor(
    Linear Color Curve=DataInterface() {propertyValues {AlphaCurve {keys [<at 0 1>]}, BlueCurve {keys [<at 0 1>]}, GreenCurve {keys [<at 0 1>]}, RedCurve {keys [<at 0 1>]}}}
    Scale Mode=2
   ) {module=ScaleColor001}
  ]
  rendererClasses [&NiagaraRibbonRendererProperties]
  renderers [{rendererClass=&NiagaraRibbonRendererProperties}]
 }
 TrailEyes {
  emitterUpdateScript [
   &EmitterState()
   &FN_SpawnAnimTrail(Number of Anim Trails=2)
  ]
  particleSpawnScript [
   &FN_InitializeAnimTrailBoneIndecies(First Bone or Socket Index=5, Second Bone or Socket Index=5)
   &FN_InitializeAnimTrailBoneIndecies(AnimTrail Selection=1, First Bone or Socket Index=6, Second Bone or Socket Index=6) {module=FN_InitializeAnimTrailBoneIndecies001}
   &FN_InitializeAnimTrail(
    AnimTrailWidth=6.6159
    Color=&MakeLinearColorFromVectorAndFloat(
     "Vector (RGB)"=&MakeVectorFromLinearColorRGB(Color=<rgba 1 1 1 1>)
     "Float (Alpha)"=&Multiply_Float(A=0.45, B=$System.FadeSpawn)
    )
    Lifetime=0.1
   )
  ]
  particleUpdateScript [
   &FN_UpdateAnimTrail() {module=FN_UpdateAnimTrail001}
   &ScaleColor(
    Scale Mode=1
    Scale RGBA=&Multiply_Vector4(
     A=&MakeVector4_FromLinearColor(LinearColor=$System.TODColor)
     B=&MakeVector4FromVectorAndFloat("Vector (RGB)"=<1 1 1>, "Float (Alpha)"=$System.FadeSystem)
    )
    ScaleRGBA=true
   ) {module=ScaleColor002}
   &ScaleColor(
    Linear Color Curve=DataInterface() {propertyValues {AlphaCurve {keys [<at 0 1>]}, BlueCurve {keys [<at 0 0.9911>]}, GreenCurve {keys [<at 0 0.9911>]}, RedCurve {keys [<at 0 0.9911>]}}}
    Scale Mode=2
   ) {module=ScaleColor001}
   /SetVariables(
    Particles.ColHighlight:LinearColor=&MultiplyLinearColor(
     Linear Color=&MaskLinearColor(
      Mask Source=1
      Mask Parameter=$System.intVariant
      Mask=0
      Masked Linear Color=<rgba 0 0.9044 2 1>
      Passthrough Linear Color=<rgba 1.3438 1.6405 2 0.3>
     )
     Scale Factor=&MakeVector4_FromLinearColor(LinearColor=$Particles.Color)
    )
    Particles.ColTrail:LinearColor=&MultiplyLinearColor(
     Linear Color=&MaskLinearColor(
      Mask Source=1
      Mask Parameter=$System.intVariant
      Mask=0
      Masked Linear Color=<rgba 0.0253 0.1205 0.5826 2>
      Passthrough Linear Color=<rgba 0.4526 0.4599 0.4957 0.9>
     )
     Scale Factor=&MakeVector4_FromLinearColor(LinearColor=$Particles.Color)
    )
   ) {isSetParameters=true}
  ]
  rendererClasses [&NiagaraRibbonRendererProperties]
  renderers [
   {rendererClass=&NiagaraRibbonRendererProperties}
   {rendererClass=&NiagaraRibbonRendererProperties, rendererIndex=1}
  ]
 }
 TrailHead {
  emitterUpdateScript [
   &EmitterState()
   &FN_SpawnAnimTrail()
  ]
  particleSpawnScript [
   &FN_InitializeAnimTrailBoneIndecies(First Bone or Socket Index=1, Second Bone or Socket Index=0)
   &FN_InitializeAnimTrail(
    Color=&MakeLinearColorFromVectorAndFloat(
     "Vector (RGB)"=&MakeVectorFromLinearColorRGB(Color=<rgba 1 1 1 1>)
     "Float (Alpha)"=&Multiply_Float(
      A=&MaskFloat(Mask Source=1, Mask Parameter=$System.intVariant, Mask=0, Masked Float=0.45, Passthrough Float=0.9)
      B=$System.FadeSpawn
     )
    )
    Lifetime=0.12
   )
  ]
  particleUpdateScript [
   &FN_UpdateAnimTrail() {module=FN_UpdateAnimTrail001}
   &ScaleColor(
    Scale Mode=1
    Scale RGBA=&Multiply_Vector4(
     A=&MakeVector4_FromLinearColor(LinearColor=$System.TODColor)
     B=&MakeVector4FromVectorAndFloat("Vector (RGB)"=<1 1 1>, "Float (Alpha)"=$System.FadeSystem)
    )
    ScaleRGBA=true
   ) {module=ScaleColor002}
   &ScaleColor(
    Linear Color Curve=DataInterface() {propertyValues {AlphaCurve {keys [<at 0.1724 1>,<at 0.9717 0>]}, BlueCurve {keys [<at 0 1>]}, GreenCurve {keys [<at 0 1>]}, RedCurve {keys [<at 0 1>]}}}
    Scale Mode=2
   ) {module=ScaleColor001}
  ]
  rendererClasses [&NiagaraRibbonRendererProperties]
  renderers [{rendererClass=&NiagaraRibbonRendererProperties}]
 }
}
summary {
 systemName=NS_Pickaxe_FunkyFeather_Cap_Trail
 emitters [
  {emitterName=TrailHead, rendererClasses [&NiagaraRibbonRendererProperties]}
  {emitterName=TrailBody, rendererClasses [&NiagaraRibbonRendererProperties]}
  {emitterName=Notes, rendererClasses [&NiagaraMeshRendererProperties]}
  {emitterName=TrailEyes, rendererClasses [&NiagaraRibbonRendererProperties]}
 ]
 userVariables [
  User.bIsFrontEnd {type=NiagaraBool}
  User.bIsFrontEndPreview {type=NiagaraBool}
  User.RibbonIDOffset {type=NiagaraInt32}
  User.SwingSectionIndex {type=NiagaraInt32}
  User.Variant {type=NiagaraFloat, defaultValue=0, description=""}
 ]
}
