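# Parent material for the FunkyFeather pickaxe cosmetic, giving it a shimmering diamond look within the 2023 character-parent shading framework.
# Between the parent's PreFX and PostFX functions, it chains MF_NewDiamond, reused from the PiperShelf Empress character, into MF_DiamondShimmer.
# The UsePickaxeBeak static switch optionally adds the FunkyFeather Beak character's interior function, masked by the red channel of the Beak pickaxe's TA texture.
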Material /BRCosmetics/Weapons/FORT_Melee/Pickaxe_FunkyFeather_Cap/Materials/M_FunkyFeather_Diamond_Parent
graph {
 properties {
  bAutomaticallySetUsageInEditor=false
  bUseMaterialAttributes=true
  bUsedWithMeshDeformer=true
  bUsedWithMorphTargets=true
  bUsedWithSkeletalMesh=true
  pixelDepthOffsetMode=PDOM_Legacy
  shadingModel=MSM_FromMaterialExpression
  usedShadingModels=Default Lit | Subsurface | Clear Coat | Subsurface Profile | Cloth
 }
 nodes [
  n1 = MaterialFunctionCall() {materialFunction=&MF_CharacterParent_2023_PreFX}
  n2 = MaterialFunctionCall!(
   In=MaterialFunctionCall!(Material=@n1.Result, TechArtMask=@n1.Result) {materialFunction=/BRCosmetics/Characters/Player/Female/Medium/Bodies/F_MED_PiperShelf_Empress/Materials/MF_NewDiamond}.Material
   TechArtMask=@n1.Tech Art Mask
  ) {materialFunction=&/MF_DiamondShimmer}
  n3 = TextureSample() {samplerType=SAMPLERTYPE_Masks, texture=/BRCosmetics/Weapons/FORT_Melee/Pickaxe_FunkyFeather_Beak/Textures/T_FunkyFeather_Beak_Pickaxe_TA}
  n4 = MaterialFunctionCall!(Material=@n2.Result, Mask=@n3.R) {materialFunction=/BRCosmetics/Characters/Player/Male/Medium/Bodies/M_MED_FunkyFeather_Cap/Skins/Beak/Materials/MF_FunkyFeather_Beak_Interior}
  MaterialFunctionCall(
   Material=StaticSwitchParameter(True=@n4.Result, False=@n2.Result) {parameterName=UsePickaxeBeak}
   M=@n1.M
   Pre Skinned Local Position=@n1.M
   Distance Blend=@n1.M
   Fresnel=@n1.M
  ) {materialFunction=&MF_CharacterParent_2023_PostFX}
 ]
}
