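# Animation Blueprint for the FunkyFeather Cap pickaxe mesh that flags its material when the cosmetic is displayed in the frontend.
# The AnimGraph passes the input pose straight to the output, adding no animation of its own.
# On each animation update, it gets the part's skeletal mesh component, checks for frontend display and, if true, sets the IsInFE scalar on its dynamic material instance.
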AnimBlueprint /BRCosmetics/Weapons/FORT_Melee/Pickaxe_FunkyFeather_Cap/Meshes/M_MED_FunkyFeather_Cap_Axe_AnimBP
target_skeleton=&/M_MED_FunkyFeather_Cap_Axe_Skeleton
graphs [
 AnimGraph {nodes [
   AnimGraphNode_SubInput_0 = Input Pose()
   Output Pose(Result=@AnimGraphNode_SubInput_0) {layerGroup=None}
  ]}
 EventGraph {nodes [
   Event BlueprintUpdateAnimation()
   n1 = GetPartSkeletalMeshComponent()
   > Branch(Condition=IsCosmeticObjectDisplayedInFrontendContext!(CosmeticObject:Object Reference=@n1) -> (ReturnValue:bool))
   > n2 = Get Dynamic Material Instance(SkeletalMeshComponent=@n1)
   > SetScalarParameterValue(self=@n2, ParameterName=IsInFE, Value=1)
  ]}
]
variables []
